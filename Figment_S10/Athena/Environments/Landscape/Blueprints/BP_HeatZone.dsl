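; BP_HeatZone is a landscape Blueprint that drives a heat-related effect through shared landscape material parameters.
; Its construction script samples the HeatCurve float curve at the HeatAmount value.
; It writes the result to the ColorTestAmount scalar in the MPC_Landscape material parameter collection, so the value applies globally to materials reading that collection rather than per instance.
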(fn ConstructionScript
  (Rendering|Material|SetScalarParameterValue "/Game/Athena/Environments/Landscape/MPC/MPC_Landscape.MPC_Landscape" "ColorTestAmount" (Math|Curves|GetFloatValue (Variables|Default|GetHeatCurve) (Variables|Default|GetHeatAmount)) :!class /Script/Engine.KismetMaterialLibrary))
